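; Burst gameplay cue that plays bite feedback when the ThornRipper hookshot projectile impacts a pawn.
; OnBurst flags the first particle component for a bite mark and whether the instigator is locally controlled.
; It then calls PlayBite, passing the hit actor as target and the cue target as instigator.
; PlayBite plays the ChainWhip bite sound at the cue location, setting audio parameters for whether either pawn is locally viewed.
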(fn OnBurst(EitherVFX-preferthisversion)
  (params (MyTarget :type Actor%20Object%20Reference) (Parameters :type Gameplay%20Cue%20Parameters%20Structure%20%28by%20ref%29) (ParticleComponents :type Array%20of%20FXSystem%20Component%20Object%20References) (AudioComponents :type Array%20of%20Audio%20Component%20Object%20References) (BurstCameraShakeInstance :type Camera%20Shake%20Base%20Object%20Reference) (BurstDecalInstance :type Decal%20Actor%20Object%20Reference))
  (bind (_bblockinghit _binitialoverlap _time _distance _location _impactpoint _normal _impactnormal _physmat _hitactor) (Collision|BreakHitResult (Ability|GameplayCue|GetHitResult Parameters)))
  (bind (_normalizedmagnitude _rawmagnitude _effectcontext _matchedtagname _originaltag _aggregatedsourcetags _aggregatedtargettags _location_2) (Ability|GameplayCue|BreakGameplayCueParameters (Variables|GetParameters)))
  (bind _output (Utilities|Array|Get(acopy) ParticleComponents))
  (Effects|Components|ParticleSystem|SetBoolParameter _output "IsLocalPlayer" (Ability|GameplayCue|IsInstigatorLocallyControlled Parameters))
  (Effects|Components|ParticleSystem|SetBoolParameter _output "BiteMarkOn" true)
  (bind _asfort_pawn (Utilities|Casting|CastToFortPawn (Variables|GetMyTarget)))
  (bind _asfort_pawn_2 (Utilities|Casting|CastToFortPawn _hitactor))
  (Default|PlayBite _asfort_pawn_2 _asfort_pawn "(TagName=\"SoundLibrary.SunRose.ChainWhip.Impact.Bite\")" _location_2)
  (return))

(fn PlayBite
  (params (Target :type Fort%20Pawn%20Object%20Reference) (Instigator :type Fort%20Pawn%20Object%20Reference) (EventName :type Gameplay%20Tag%20Structure) (Location :type Vector :default "0, 0, 0"))
  (bind _returnvalue (WorldSubsystems|GetSoundLibrarySubsystem Instigator))
  ;[0.007682 0.1219 0.1354] PlaySound: Play the desired event name sound at the provided location.
  (Utilities|IsValid _returnvalue
    (:Is%20Valid
      (bind _returnvalue_2 (SoundLibrary|PlaySoundSimple _returnvalue Instigator (Utilities|Struct|MakeSimpleSoundSettings EventName false false true true 1.0 1.0 0.0 0 0 "None" 0 "KeepWorldPosition" Location) :!class /Script/SoundLibrary.SoundLibrarySubsystem))
      (for _array_element (.AudioComponents (Utilities|Struct|BreakSoundLibraryPlayResult _returnvalue_2))
        (Audio|Parameter|SetBooleanParameter _array_element "istarget" (Pawn|IsLocallyViewed (Variables|GetTarget)) :!class /Script/AudioExtensions.AudioParameterControllerInterface)
        (Audio|Parameter|SetBooleanParameter _array_element "isinstigator" (Pawn|IsLocallyViewed (Variables|GetInstigator)) :!class /Script/AudioExtensions.AudioParameterControllerInterface)))))
